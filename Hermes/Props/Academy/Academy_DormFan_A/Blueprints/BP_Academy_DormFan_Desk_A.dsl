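; Desk fan prop for the Academy dorm, assembled from a main static mesh plus separate motor and blades meshes.
; Construction calls UpdateFanParts, which moves the motor to the main mesh's first socket, then the blades to the motor's first socket.
; It then applies the tint from the first building texture data entry to both blades and motor.
; BeginPlay, overlap and Tick events are empty.
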(event EventBeginPlay)

(event Collision|EventActorBeginOverlap
  (params (OtherActor :type Actor%20Object%20Reference)))

(event EventTick
  (params (DeltaSeconds :type Float%20%28single-precision%29)))

(fn UpdateFanParts
  (bind _sm_academy_dormfanmotor_a (Variables|Default|GetSM_Academy_DormFanMotor_A))
  (bind _returnvalue (Transformation|GetAllSocketNames _sm_academy_dormfanmotor_a))
  (bind _staticmeshcomponent (Variables|Debug|GetStaticMeshComponent))
  (bind _returnvalue_2 (Transformation|GetAllSocketNames _staticmeshcomponent))
  (if (Utilities|Array|IsValidIndex _returnvalue_2)
    (Transformation|SetWorldLocation (Variables|Default|GetSM_Academy_DormFanMotor_A) (Transformation|GetSocketLocation _staticmeshcomponent (Utilities|Array|Get(acopy) _returnvalue_2)) false false)
    (if (Utilities|Array|IsValidIndex _returnvalue)
      (Transformation|SetWorldLocation (Variables|Default|GetSM_Academy_DormFanBlades_A) (Transformation|GetSocketLocation _sm_academy_dormfanmotor_a (Utilities|Array|Get(acopy) _returnvalue)) false false))))

(fn ConstructionScript
  (|CallParentFunction)
  (CallFunction|UpdateFanParts)
  (bind _outtexturedataarray (Building|GetBuildingTextureData))
  (if (Utilities|Array|IsValidIndex _outtexturedataarray)
    (Utilities|FlowControl|Sequence
      (:then_0
        (Default|SetColoronSecondaryMesh (Math|Conversions|ToLinearColor(Color) (Variables|/Script/FortniteGame.BuildingTextureData|GetTintColor (Utilities|Array|Get(acopy) _outtexturedataarray))) (Variables|Default|GetSM_Academy_DormFanBlades_A)))
      (:then_1
        (Default|SetColoronSecondaryMesh (Math|Conversions|ToLinearColor(Color) (Variables|/Script/FortniteGame.BuildingTextureData|GetTintColor (Utilities|Array|Get(acopy) _outtexturedataarray))) (Variables|Default|GetSM_Academy_DormFanMotor_A))))))
